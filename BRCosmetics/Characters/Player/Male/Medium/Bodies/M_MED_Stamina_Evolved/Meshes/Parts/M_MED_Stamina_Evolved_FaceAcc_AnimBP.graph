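# Post-process style hair animation for the Stamina Evolved male face accessory, copying the parent mesh pose and adding secondary motion.
# Head-relative AnimDynamics on four bangs bones feeds a RigidBody hair simulation, cached as prePose.
# For the local player, top-hair bones blend to scaled, offset transforms while targeting with an unholstered weapon, except trap, rocket launcher and GoingCommando types.
# The bangs_03 dynamics node is left unconnected.
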AnimBlueprint /BRCosmetics/Characters/Player/Male/Medium/Bodies/M_MED_Stamina_Evolved/Meshes/Parts/M_MED_Stamina_Evolved_FaceAcc_AnimBP
target_skeleton=&/M_MED_Stamina_Evolved_FaceAcc_Skeleton
graphs [
 AnimGraph {nodes [
   Anim Dynamics - Bone: dyn_bangs_03_MSTR!(ComponentPose, Alpha=1, GravityScale:float=1, GravityOverride:Vector="0,0,-980", ExternalForce="0,0,0") -> (Pose) {
    angularDampingOverride=1
    angularSpringConstant=200
    bAngularSpring=true
    bGravityOverrideInSimSpace=true
    bOverrideAngularDamping=true
    simulationSpace=RootRelative
    alphaBoolBlend {blendOption=Linear}
    boundBone {boneName=dyn_bangs_03_MSTR}
    componentAppliedLinearAccClamp=<1000 1000 1000>
    componentLinearAccScale=<0.15 0.15 0.15>
    componentLinearVelScale=<0.5 0.5 0.5>
    gravityOverride {z=-980}
    physicsBodyDefinitions [{
      sphereCollisionRadius=0
      boundBone {boneName=dyn_bangs_03_MSTR}
      boxExtents=<10 12 16>
      constraintSetup {angularLimitsMax=<0 10 10>, angularLimitsMin=<0 -5 -10>, angularTarget=<0 0 0>}
      localJointOffset=<1 -0.1 -2>
     }]
    relativeSpaceBone {boneName=head}
   }
   n1 = Get scaleDownAlpha() -> (scaleDownAlpha:bool)
   n2 = "Transform (Modify) Bone - Bone: dyn_top_hair_01"!(
    ComponentPose=Local To Component(LocalPose=Use cached pose 'prePose'())
    bAlphaBoolEnabled:bool=@n1
    Translation="0,0,0"
    Rotation="0,0,0"
    Scale="0.75,0.75,0.75"
   ) -> (Pose) {
    alphaInputType=Bool
    scaleMode=BMM_Additive
    scaleSpace=BCS_BoneSpace
    translationMode=BMM_Additive
    translationSpace=BCS_BoneSpace
    alphaBoolBlend {blendInTime=0.1, blendOption=Linear, blendOutTime=0.12}
    boneToModify {boneName=dyn_top_hair_01}
    scale=<0.75 0.75 0.75>
   }
   n3 = "Transform (Modify) Bone - Bone: dyn_top_hair_02"!(ComponentPose=@n2, bAlphaBoolEnabled:bool=@n1, Translation="-5,0,0", Rotation="0,0,0", Scale="0.75,0.75,0.75") -> (Pose) {
    alphaInputType=Bool
    scaleMode=BMM_Additive
    scaleSpace=BCS_BoneSpace
    translationMode=BMM_Additive
    translationSpace=BCS_BoneSpace
    alphaBoolBlend {blendInTime=0.1, blendOption=Linear, blendOutTime=0.12}
    boneToModify {boneName=dyn_top_hair_02}
    scale=<0.75 0.75 0.75>
    translation {x=-5}
   }
   n4 = "Transform (Modify) Bone - Bone: dyn_top_hair_03"!(ComponentPose=@n3, bAlphaBoolEnabled:bool=@n1, Translation="-1,0,0", Rotation="0,0,0", Scale="0.75,0.75,0.75") -> (Pose) {
    alphaInputType=Bool
    scaleMode=BMM_Additive
    scaleSpace=BCS_BoneSpace
    translationMode=BMM_Additive
    translationSpace=BCS_BoneSpace
    alphaBoolBlend {blendInTime=0.1, blendOption=Linear, blendOutTime=0.12}
    boneToModify {boneName=dyn_top_hair_03}
    scale=<0.75 0.75 0.75>
    translation {x=-1}
   }
   n5 = "Transform (Modify) Bone - Bone: dyn_top_hair_04"!(ComponentPose=@n4, bAlphaBoolEnabled:bool=@n1, Translation="-3,1.462,-0.000", Rotation="-5.000,-0.000,0.000", Scale="0.7,0.7,0.7") -> (Pose) {
    alphaInputType=Bool
    rotationMode=BMM_Additive
    rotationSpace=BCS_BoneSpace
    scaleMode=BMM_Additive
    scaleSpace=BCS_BoneSpace
    translationMode=BMM_Additive
    translationSpace=BCS_BoneSpace
    alphaBoolBlend {blendInTime=0.1, blendOption=Linear, blendOutTime=0.12}
    boneToModify {boneName=dyn_top_hair_04}
    rotation=<rot -5 0 0>
    scale=<0.7 0.7 0.7>
    translation=<-3 1.462>
   }
   n6 = "Transform (Modify) Bone - Bone: dyn_top_hair_05"!(ComponentPose=@n5, bAlphaBoolEnabled:bool=@n1, Translation="0,0,0", Rotation="0,0,0", Scale="0.8,0.85,0.85") -> (Pose) {
    alphaInputType=Bool
    rotationMode=BMM_Additive
    rotationSpace=BCS_BoneSpace
    scaleMode=BMM_Additive
    scaleSpace=BCS_BoneSpace
    translationMode=BMM_Additive
    translationSpace=BCS_BoneSpace
    alphaBoolBlend {blendInTime=0.1, blendOption=Linear, blendOutTime=0.12}
    boneToModify {boneName=dyn_top_hair_05}
    scale=<0.8 0.85 0.85>
   }
   Output Pose(Result=Blend Poses by bool(
     BlendPose_0=Component To Local(ComponentPose=@n6)
     BlendPose_1=Use cached pose 'prePose'()
     bActiveValue=Get isLocalPlayer() -> (isLocalPlayer:bool)
    ) {blendType=Linear}) {layerGroup=None}
   n7 = Anim Dynamics - Bone: dyn_bangs_01_MSTR!(
    ComponentPose=Local To Component(LocalPose=Copy Pose From Mesh() {bUseAttachedParent=true})
    Alpha=0.8
    GravityScale:float=1
    GravityOverride:Vector="0,0,-980"
    ExternalForce="0,0,0"
   ) -> (Pose) {
    angularDampingOverride=0.9999
    angularSpringConstant=200
    bAngularSpring=true
    bGravityOverrideInSimSpace=true
    bOverrideAngularDamping=true
    bUseGravityOverride=true
    simulationSpace=RootRelative
    alphaBoolBlend {blendOption=Linear}
    boundBone {boneName=dyn_bangs_01_MSTR}
    componentAppliedLinearAccClamp=<1000 1000 1000>
    componentLinearAccScale=<0.15 0.15 0.15>
    componentLinearVelScale=<0.5 0.5 0.5>
    gravityOverride {z=-980}
    physicsBodyDefinitions [{
      sphereCollisionRadius=0
      boundBone {boneName=dyn_bangs_01_MSTR}
      boxExtents=<10 16 12>
      constraintSetup {angularLimitsMax=<0 0 10>, angularLimitsMin=<0 0 -10>, angularTarget=<0 0 0>}
      localJointOffset=<1 0 -0.2>
     }]
    relativeSpaceBone {boneName=head}
   }
   n8 = Anim Dynamics - Bone: dyn_bangs_02_MSTR!(ComponentPose=@n7, Alpha=0.8, GravityScale:float=1, GravityOverride:Vector="0,0,-980", ExternalForce="0,0,0") -> (Pose) {
    angularDampingOverride=1
    angularSpringConstant=200
    bAngularSpring=true
    bGravityOverrideInSimSpace=true
    bOverrideAngularDamping=true
    bUseGravityOverride=true
    simulationSpace=RootRelative
    alphaBoolBlend {blendOption=Linear}
    boundBone {boneName=dyn_bangs_02_MSTR}
    componentAppliedLinearAccClamp=<1000 1000 1000>
    componentLinearAccScale=<0.15 0.15 0.15>
    componentLinearVelScale=<0.5 0.5 0.5>
    gravityOverride {z=-980}
    physicsBodyDefinitions [{
      sphereCollisionRadius=0
      boundBone {boneName=dyn_bangs_02_MSTR}
      boxExtents=<10 16 12>
      constraintSetup {angularLimitsMax=<0 15 10>, angularLimitsMin=<0 -5 -10>, angularTarget=<0 0 0>}
      localJointOffset=<2 -0.3 -2>
     }]
    relativeSpaceBone {boneName=head}
   }
   n9 = Anim Dynamics - Bone: dyn_bangs_04_MSTR!(ComponentPose=@n8, Alpha=0.8, GravityScale:float=1, GravityOverride:Vector="0,0,-980", ExternalForce="0,0,0") -> (Pose) {
    angularDampingOverride=0.9999
    angularSpringConstant=200
    bAngularSpring=true
    bGravityOverrideInSimSpace=true
    bOverrideAngularDamping=true
    bUseGravityOverride=true
    simulationSpace=RootRelative
    alphaBoolBlend {blendOption=Linear}
    boundBone {boneName=dyn_bangs_04_MSTR}
    componentAppliedLinearAccClamp=<1000 1000 1000>
    componentLinearAccScale=<0.15 0.15 0.15>
    componentLinearVelScale=<0.5 0.5 0.5>
    gravityOverride {z=-980}
    physicsBodyDefinitions [{
      sphereCollisionRadius=0
      boundBone {boneName=dyn_bangs_04_MSTR}
      boxExtents=<10 16 12>
      constraintSetup {angularLimitsMax=<0 5 20>, angularLimitsMin=<0 -5 -20>, angularTarget=<0 0 0>}
      localJointOffset=<1 0 -0.1>
     }]
    relativeSpaceBone {boneName=head}
   }
   n10 = Anim Dynamics - Bone: dyn_bangs_05_MSTR!(ComponentPose=@n9, Alpha=0.8, GravityScale:float=1, GravityOverride:Vector="0,0,-980", ExternalForce="0,0,0") -> (Pose) {
    angularDampingOverride=0.9999
    angularSpringConstant=200
    bAngularSpring=true
    bGravityOverrideInSimSpace=true
    bOverrideAngularDamping=true
    bUseGravityOverride=true
    simulationSpace=RootRelative
    alphaBoolBlend {blendOption=Linear}
    boundBone {boneName=dyn_bangs_05_MSTR}
    componentAppliedLinearAccClamp=<1000 1000 1000>
    componentLinearAccScale=<0.15 0.15 0.15>
    componentLinearVelScale=<0.5 0.5 0.5>
    gravityOverride {z=-980}
    physicsBodyDefinitions [{
      sphereCollisionRadius=0
      boundBone {boneName=dyn_bangs_05_MSTR}
      boxExtents=<10 16 12>
      constraintSetup {angularLimitsMax=<0 5 20>, angularLimitsMin=<0 -5 -20>, angularTarget=<0 0 0>}
      localJointOffset=<1 -0.1 -0.1>
     }]
    relativeSpaceBone {boneName=head}
   }
   n11 = RigidBody(ComponentPose=@n10) {
    bClampLinearTranslationLimitToRefPose=true
    bOverrideWorldGravity=true
    overridePhysicsAsset=&/M_MED_Stamina_Evolved_FaceAcc_dyn_hair
    simulationSpace=BaseBoneSpace
    alphaBoolBlend {blendOption=Linear}
    baseBoneRef {boneName=head}
    componentAppliedLinearAccClamp=<1000 1000 1000>
    componentLinearAccScale=<0.15 0.15 0.15>
    componentLinearVelScale=<0.4 0.4 0.4>
    overrideWorldGravity {z=-700}
   }
   Save cached pose 'prePose'(Pose=Component To Local(ComponentPose=@n11)) {cacheName=prePose}
  ]}
 EventGraph {nodes [
   Event BlueprintUpdateAnimation()
   > n1 = Sequence()
   > Set FortPawn(FortPawn:Fort Pawn Object Reference=TryGetFortPawn()) -> (Output_Get:Fort Pawn Object Reference)
   n2 = Get FortPawn() -> (FortPawn:Fort Pawn Object Reference)
   > n3 = Is Valid(InputObject=@n2)
   Set weapon_visible(execute=@n3.Is Not Valid, weapon_visible:bool=false) -> (Output_Get:bool)
   n4 = Get CurrentWeapon!(self:Fort Pawn Object Reference=@n2) -> (CurrentWeapon:Fort Weapon Object Reference)
   n5 = GetCoreAnimation!(self:Fort Weapon Object Reference=@n4) -> (ReturnValue:EFortWeaponCoreAnimation Enum)
   n6 = Set notCoverCrosshair(
    execute=@n3.Is Valid
    notCoverCrosshair:bool=OR Boolean!(
     A="Equal (Enum)"(A:EFortWeaponCoreAnimation Enum=@n5, B:EFortWeaponCoreAnimation Enum=TrapPlacement)
     B="Equal (Enum)"(A=@n5, B=RocketLauncher)
     C:bool="Equal (Enum)"(A=@n5, B=GoingCommando)
    ) -> (ReturnValue)
   ) -> (Output_Get:bool)
   > n7 = Branch(Condition=@n6)
   > Set weapon_visible(weapon_visible=false)
   n8 = IsWeaponHolstered!(self:Fort Pawn Object Reference=@n2) -> (ReturnValue:bool)
   Set weapon_visible(execute=@n7.else, weapon_visible=AND Boolean(A=IsValid(Object=@n4), B=NOT Boolean(A=@n8)))
   Set isLocalPlayer(execute=@n1.then_1, isLocalPlayer:bool=IsLocalPawn!(Actor:Actor Object Reference=GetOwningActor()) -> (IsLocalPawn:bool)) -> (Output_Get:bool)
   > n9 = Branch(Condition=AND Boolean(A=Get bIsTargeting(), B=Get weapon_visible() -> (weapon_visible:bool)))
   > Set scaleDownAlpha(scaleDownAlpha:bool=true) -> (Output_Get:bool)
   Set scaleDownAlpha(execute=@n9.else, scaleDownAlpha=false)
  ]}
]
variables [
 FortPawn {type=FortPawn, default=null}
 isLocalPlayer {type=boolean}
 notCoverCrosshair {type=boolean}
 scaleDownAlpha {type=boolean}
 weapon_visible {type=boolean}
]
